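# Niagara system for the spin hit effect in the RubySnapper emote, pairing a brief sprite burst with movement-driven speed lines.
# Speedlines spawns ribbon particles per unit traveled, with a quick alpha fade.
# Minimal fires a slightly delayed burst of short-lived, randomly mirrored sprites with inherited velocity and drag.
# It has three sprite renderers, a particle-ID-based VisibilityTag, and separately animated ring and secondary size attributes.
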NiagaraSystem /BRCosmetics/Animation/Game/MainPlayer/Emotes/RubySnapper/VFX/NS_Emote_RubbySnapper_SpinHit
dependencies {
 usedDataInterfaces [
  &NiagaraDataInterfaceCurve
  &NiagaraDataInterfaceSpriteRendererInfo
  &NiagaraDataInterfaceVector2DCurve
 ]
 usedDynamicInputs []
 usedModules [&SystemState]
}
emitters {
 Minimal {
  emitterUpdateScript [
   &EmitterState(Life Cycle Mode=::NewEnumerator1, Loop Behavior=::NewEnumerator1, Loop Delay=0.09, Loop Duration=1, UseLoopDelay=true)
   &SpawnBurst_Instantaneous(Spawn Count=6, Use Spawn Probability=false, UseLoopCountLimit=false)
  ]
  particleSpawnScript [
   &InitializeParticle(
    Color Mode=::NewEnumerator0
    Lifetime=&RandomRangeFloat(Minimum=0.15, Maximum=0.25)
    Sprite Rotation Mode=::NewEnumerator2
    Sprite Size Mode=::NewEnumerator1
    Sprite UV Mode=::NewEnumerator4
    Sprite UV Scale:Vector2f=&MakeVector2D(
     X=&MakeCustomFloatFromBool(Bool=&RandomBool(), True Float=1, False Float=-1)
     Y=-1
    )
    Uniform Sprite Size Max=50
    Uniform Sprite Size Min=40
    UsePositionOffset=false
    Write Material Random=true
   )
   &InheritVelocity(Inherited Velocity Speed Limit=30, Speed Limit=true, UseSpeedThreshold=false)
   /SetVariables(Particles.VisibilityTag=&ReturnParticleIDAttributes()) {isSetParameters=true}
   &CameraOffset(Camera Offset Amount=50)
   &SubUVAnimation(
    "Sprite Renderer "=DataInterface() {propertyValues {SpriteRenderer=&self:Minimal.NiagaraSpriteRendererProperties_1}}
    SubUV Animation Mode:ENiagara_SubUVLookupModeV2=::NewEnumerator2
    UseEndFrame=false
    UseStartFrame=false
   )
  ]
  particleUpdateScript [
   &ParticleState()
   &ScaleSpriteSize(Uniform Curve Sprite Scale=DataInterface() {propertyValues {Curve {keys [
        {arriveTangent=7.7405, interpMode=RCIM_Cubic, leaveTangent=7.7405, tangentMode=RCTM_User, time=0, value=0}
        {interpMode=RCIM_Cubic, tangentMode=RCTM_User, time=0.2, value=1}
        <at 1 0>
       ]}}})
   &Drag(Drag=3, Use Linear Drag=true, Use Rotational Drag=false)
   &SolveForcesAndVelocity(Clamp Velocity=false, Limit Acceleration=false)
   /SetVariables(
    Particles.RingSpriteSize:Vector2f=&Multiply_Vector2D_ByFloat(
     Float=&FloatFromCurve(FloatCurve=DataInterface() {propertyValues {Curve {keys [{arriveTangent=0.8668, interpMode=RCIM_Cubic, leaveTangent=0.8668, tangentMode=RCTM_User, time=0, value=1}, <at 1 1.5>]}}})
     Vector2D=<40 40>
    )
    Particles.SpriteSize2=&Vector2DFromCurve(
     Vector2Curve=DataInterface() {propertyValues {
       XCurve {keys [
         {arriveTangent=8.1517, interpMode=RCIM_Cubic, leaveTangent=8.1517, tangentMode=RCTM_User, time=0, value=0.2}
         {arriveTangent=0.2469, interpMode=RCIM_Cubic, leaveTangent=0.2469, tangentMode=RCTM_User, time=0.15, value=1}
         <at 1 1>
        ]}
       YCurve {keys [
         {arriveTangent=8.1517, interpMode=RCIM_Cubic, leaveTangent=8.1517, tangentMode=RCTM_User, time=0, value=0.5}
         {arriveTangent=-0.0461, interpMode=RCIM_Cubic, leaveTangent=-0.0461, tangentMode=RCTM_User, time=0.15, value=1}
         <at 1 1>
        ]}
      }}
     ScaleCurve=&RandomRangeVector2D(Minimum=<35 55>, Maximum=<45 65>)
    )
   ) {isSetParameters=true}
   &ScaleColor(
    Scale Alpha=&FloatFromCurve(FloatCurve=DataInterface() {propertyValues {Curve {keys [{arriveTangent=-1.3634, interpMode=RCIM_Cubic, leaveTangent=-1.3634, tangentMode=RCTM_User, time=0, value=1}, <at 1 0>]}}})
    ScaleA=true
    ScaleRGB=false
   )
  ]
  rendererClasses [&NiagaraSpriteRendererProperties]
  renderers [
   {rendererClass=&NiagaraSpriteRendererProperties}
   {rendererClass=&NiagaraSpriteRendererProperties, rendererIndex=1}
   {rendererClass=&NiagaraSpriteRendererProperties, rendererIndex=2}
  ]
 }
 Speedlines {
  emitterUpdateScript [
   &EmitterState(Life Cycle Mode=::NewEnumerator1, Loop Behavior=::NewEnumerator1, Loop Duration=0.3, UseLoopDelay=false)
   &SpawnPerUnit(Module.Use Max Movement Threshold=true, Module.Use Movement Tolerance=false, Spawn Spacing=10, Use Spawn Probability=false)
  ]
  particleSpawnScript [
   &InitializeParticle(
    Color Mode=::NewEnumerator0
    Lifetime=0.2
    Ribbon Width=10
    Ribbon Width Mode=::NewEnumerator1
    UsePositionOffset=false
    Write Material Random=true
   )
   &ShapeLocation(MaskSpawnGroup=false, Override Seed=false, Sphere Radius=0, UseOwnerScale=true)
  ]
  particleUpdateScript [
   &ParticleState()
   &ScaleColor(
    Scale Alpha=&FloatFromCurve(FloatCurve=DataInterface() {propertyValues {Curve {keys [
         {interpMode=RCIM_Cubic, tangentMode=RCTM_User, time=0.1552, value=0}
         {interpMode=RCIM_Cubic, tangentMode=RCTM_User, time=0.2, value=1}
         {tangentMode=RCTM_User, time=1, value=0}
        ]}}})
    ScaleA=true
    ScaleRGB=true
   )
  ]
  rendererClasses [&NiagaraRibbonRendererProperties]
  renderers [{rendererClass=&NiagaraRibbonRendererProperties}]
 }
}
summary {
 systemName=NS_Emote_RubbySnapper_SpinHit
 emitters [
  {emitterName=Speedlines, rendererClasses [&NiagaraRibbonRendererProperties]}
  {emitterName=Minimal, rendererClasses [&NiagaraSpriteRendererProperties]}
 ]
 userVariables []
}
