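# Niagara system for the appear effect of the RubySnapper emote, a short burst of colored smoke puffs.
# SpawnSmoke bursts sprites in a small sphere, each tinted randomly from a six-color palette and launched fast, then slowed by heavy drag.
# Over each short life, curves grow the sprites, advance SubUV frames from a random start and drive a falling dynamic material parameter.
# The airglow, sparkle and toon impact smoke emitters list no renderers.
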NiagaraSystem /BRCosmetics/Animation/Game/MainPlayer/Emotes/RubySnapper/VFX/NS_Emote_RubySnapper_Appear
dependencies {
 usedDataInterfaces [
  &NiagaraDataInterfaceArrayColor
  &NiagaraDataInterfaceCurve
  &NiagaraDataInterfaceSpriteRendererInfo
  &NiagaraDataInterfaceArrayFloat
 ]
 usedDynamicInputs [&FNTimeOfDayColor, &SelectFloatFromArray]
 usedModules [
  &SystemState
  &self:SystemSpawnScript.SystemScriptSource.SystemScriptGraph.NiagaraNodeAssignment_0.SetVariables_#1
 ]
}
emitters {
 AirglowLWE {rendererClasses [], renderers []}
 Sparkle_LG_LWE {rendererClasses [], renderers []}
 SparklesLWE {rendererClasses [], renderers []}
 SpawnSmoke {
  emitterUpdateScript [
   &EmitterState()
   &SpawnBurst_Instantaneous(Spawn Count=20, Use Spawn Probability=false, UseLoopCountLimit=false)
  ]
  particleSpawnScript [
   &InitializeParticle(
    Color=&SelectLinearColorFromArray(
     Array Sampling Mode=::NewEnumerator1
     Direct Array Index=&RandomRangeInt(Minimum=0, Maximum=5)
     Color Selection Array=DataInterface() {propertyValues {ColorData [
        <rgba 0.2091 0.4564 0.1964 1>
        <rgba 0.2291 0.2742 0.6376 1>
        <rgba 0.8796 0.7814 0.2025 1>
        <rgba 0.8308 0.2882 0.5875 1>
        <rgba 0.8796 0.2575 0.2751 1>
        <rgba 0.6867 0.2169 0.2067 1>
       ]}}
     Random Seed=?
    )
    Lifetime=&RandomRangeFloat(Minimum=0.3, Maximum=1)
    Sprite Rotation Mode=::NewEnumerator2
    Sprite Size Mode=::NewEnumerator3
    Sprite UV Mode=::NewEnumerator3
    Uniform Sprite Size=50
    UsePositionOffset=false
    Write Material Random=true
   )
   &ShapeLocation(MaskSpawnGroup=false, Override Seed=false, Sphere Radius=20, UseOwnerScale=true)
   &AddVelocity(
    Module.Use Falloff Distance=false
    Origin Offset=<0 0 -1>
    Override Seed=false
    Velocity Mode=::NewEnumerator1
    Velocity Speed=&RandomRangeFloat(Minimum=200, Maximum=400)
   )
  ]
  particleUpdateScript [
   &ParticleState()
   &ScaleMeshSize(Scale Factor=&Multiply_VectorByFloat(
     Vector=<1 1 1>
     Float=&FloatFromCurve(
      FloatCurve=DataInterface() {propertyValues {Curve {
         defaultValue=1
         keys [<at 0 1 RCIM_Cubic>, <at 0.2225 1.7>, {arriveTangent=0.3858, interpMode=RCIM_Cubic, leaveTangent=0.3858, time=0.3582, value=1.85}, <at 1 2>]
        }}}
      Scale Curve=&RandomRangeFloat(Minimum=1, Maximum=1.3)
     )
    ))
   &ScaleSpriteSize(Uniform Curve Sprite Scale=DataInterface() {propertyValues {Curve {keys [
        {arriveTangent=1.1204, interpMode=RCIM_Cubic, leaveTangent=1.1204, tangentMode=RCTM_User, time=0, value=0.5}
        {tangentMode=RCTM_User, time=1, value=1}
       ]}}})
   &Drag(Drag=&RandomRangeFloat(Minimum=9, Maximum=13), Use Linear Drag=true, Use Rotational Drag=true)
   &SolveForcesAndVelocity(Clamp Velocity=false, Limit Acceleration=false)
   &DynamicMaterialParameters(
    Index 0 Param 1=&FloatFromCurve(FloatCurve=DataInterface() {propertyValues {Curve {defaultValue=1, keys [<at 0 1 RCIM_Cubic>, {arriveTangent=-1.9, interpMode=RCIM_Cubic, leaveTangent=-1.9, time=0.2, value=1}, <at 1 -0.9>]}}})
    Index 0 Param 4=0.5
    Module.Index0Float0Write=true
    Module.Index0Float1Write=false
    Module.Index0Float2Write=false
    Module.Index0Float3Write=true
   )
   &SubUVAnimation(
    Random Start Frame=true
    "Sprite Renderer "=DataInterface() {propertyValues {SpriteRenderer=&self:SpawnSmoke_0.NiagaraSpriteRendererProperties_0}}
    Start Frame Offset:NiagaraInt32=&Multiply_Int(A=0, B=16)
    SubUV Animation Mode:ENiagara_SubUVLookupModeV2=::NewEnumerator1
    SubUV Curve=DataInterface() {propertyValues {Curve {keys [
        {arriveTangent=4.1785, interpMode=RCIM_Cubic, leaveTangent=4.1785, tangentMode=RCTM_User, time=0, value=0}
        {arriveTangent=1.1449, interpMode=RCIM_Cubic, leaveTangent=1.1449, tangentMode=RCTM_User, time=0.2, value=0.6376}
        {arriveTangent=0.2086, interpMode=RCIM_Cubic, leaveTangent=0.2086, tangentMode=RCTM_User, time=1, value=1}
       ]}}}
    UseEndFrame=false
    UseStartFrame=false
   )
  ]
  rendererClasses [&NiagaraSpriteRendererProperties]
  renderers [{rendererClass=&NiagaraSpriteRendererProperties}]
 }
 ToonImpactSmoke {rendererClasses [], renderers []}
}
summary {
 systemName=NS_Emote_RubySnapper_Appear
 emitters [
  {emitterName=AirglowLWE, rendererClasses []}
  {emitterName=SparklesLWE, rendererClasses []}
  {emitterName=ToonImpactSmoke, rendererClasses []}
  {emitterName=SpawnSmoke, rendererClasses [&NiagaraSpriteRendererProperties]}
  {emitterName=Sparkle_LG_LWE, rendererClasses []}
 ]
 userVariables []
}
