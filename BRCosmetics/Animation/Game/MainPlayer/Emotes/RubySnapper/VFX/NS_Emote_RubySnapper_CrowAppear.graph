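# Niagara effect for a crow appearing in the RubySnapper emote, built from dark feather and smoke emitters.
# Feathers spawns randomized spinning meshes that drift down and shrink; SpawnSmoke bursts gray flipbook sprites that grow and slow under drag.
# Both follow source movement, while front-end user bools adjust feather offsets, velocity and acceleration for menu display.
# AirglowLWE, SparklesLWE, ToonImpactSmoke and the disabled ToonSmoke list no renderers or modules.
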NiagaraSystem /BRCosmetics/Animation/Game/MainPlayer/Emotes/RubySnapper/VFX/NS_Emote_RubySnapper_CrowAppear
dependencies {
 usedDataInterfaces [
  &NiagaraDataInterfaceCurve
  &NiagaraDataInterfaceArrayColor
  &NiagaraDataInterfaceSpriteRendererInfo
  &NiagaraDataInterfaceArrayFloat
 ]
 usedDynamicInputs [&FNTimeOfDayColor, &SelectFloatFromArray]
 usedModules [
  &SystemState
  &self:SystemSpawnScript.SystemScriptSource.SystemScriptGraph.NiagaraNodeAssignment_0.SetVariables_#1
 ]
}
emitters {
 AirglowLWE {rendererClasses [], renderers []}
 Feathers {
  emitterUpdateScript [
   &EmitterState(Life Cycle Mode=::NewEnumerator1, Loop Behavior=::NewEnumerator1, Loop Delay=0.1, Loop Duration=0.4, UseLoopDelay=true)
   &SpawnRate(SpawnRate=5, Use Spawn Probability=false)
   &SpawnBurst_Instantaneous(
    Spawn Count=&RandomRangeInt(Minimum=6, Maximum=9)
    Use Spawn Probability=false
    UseLoopCountLimit=false
   )
  ]
  particleSpawnScript [
   &InitializeParticle(
    Color=&MakeCustomLinearColorFromBool(Bool=&RandomBool(), True Color=<rgba 0 0 0 1>, False Color=<rgba 0.0156 0.0156 0.0156 1>)
    Lifetime=&RandomRangeFloat(Minimum=0.7, Maximum=1.3)
    Mesh Scale=&Multiply_Vector(
     A=$Engine.Owner.Scale
     B=/Niagara/DynamicInputs/CascadeConversion/SplitVector(
      Double Component Vector Channels:ECascadeNiagaraTwoVectorChannels=::NewEnumerator0
      Double Component:NiagaraFloat=&RandomRangeFloat(Minimum=0.2, Maximum=0.25)
      Single Component:NiagaraFloat=&RandomRangeFloat(Minimum=0, Maximum=1)
     )
    )
    Mesh Scale Mode=::NewEnumerator4
    Position Offset=&MakeCustomVectorFromBool(Bool=$User.bIsFrontEndPreview, True Vector=<0 0 125>, False Vector=<0 0 0>)
    UsePositionOffset=true
    Write Material Random=true
   )
   &ShapeLocation(
    MaskSpawnGroup=false
    Offset=&FN_FrontEndVector3Modifier(Gameplay=<0 0 0>, Front End=<0 0 -150>)
    Offset Coordinate Space=::Local
    Offset Mode=::NewEnumerator0
    Override Seed=false
    Sphere Radius=5
    Sphere Surface Distribution=1
    UseOwnerScale=true
   )
   &AddVelocity(
    Rotation Coordinate Space=::Simulation
    Velocity=&FN_FrontEndVector3Modifier(
     Gameplay=&RandomRangeVector(Minimum=<-150 -150 -50>, Maximum=<150 150 -200>)
     Front End=&RandomRangeVector(Minimum=<-10 -10 -10>, Maximum=<10 10 10>)
    )
    Velocity Speed Scale=&FloatFromCurve(
     FloatCurve=DataInterface() {propertyValues {Curve {keys [{arriveTangent=-1.853, interpMode=RCIM_Cubic, leaveTangent=-1.853, tangentMode=RCTM_User, time=0, value=1}, <at 1 0.1>]}}}
     Scale Curve=0.5
    )
   )
   &AddVelocity(
    Rotation Coordinate Space=::Simulation
    Velocity=&Multiply_VectorByFloat(
     Vector=$Particles.ShapeLocation.ShapeNormal
     Float=&MakeCustomFloatFromBool(Bool=$User.bIsFrontEnd, True Float=0, False Float=200)
    )
   ) {module=AddVelocity001}
   &DynamicMaterialParameters(
    Index 0 Param 1=&RandomRangeFloat(Minimum=0, Maximum=0.9)
    Index 0 Param 2=9
    Index 0 Param 4=&RandomRangeFloat(Minimum=0, Maximum=1)
    Module.Param0WriteEnabled=true
    Module.Param1WriteEnabled=true
    Module.Param2WriteEnabled=true
    Module.Param3WriteEnabled=true
   )
   &InitialMeshOrientation(Use Rotation Vector=false)
   /SetVariables(Particles.SpawnRotationVector=&RandomRangeVector(Minimum=<0 0 0>, Maximum=<1 1 1>)) {isSetParameters=true}
  ]
  particleUpdateScript [
   &ParticleState()
   &ScaleMeshSize(Scale Factor=&VectorFromFloat(Value=&FloatFromCurve(FloatCurve=DataInterface() {propertyValues {Curve {keys [<at 0.5 1>,<at 1 0>]}}})))
   &UpdateMeshOrientation(
    Rotation Rate=&FloatFromCurve(
     FloatCurve=DataInterface() {propertyValues {Curve {keys [{interpMode=RCIM_Cubic, tangentMode=RCTM_User, time=0, value=1}, <at 1 0>]}}}
     Scale Curve=&RandomRangeFloat(Minimum=4, Maximum=1)
    )
    Rotation Vector=$Particles.SpawnRotationVector
   )
   &Drag(Drag=3, Use Linear Drag=true, Use Rotational Drag=false)
   &InheritSourceMovement(Applied Position Scale=<0.75 0.75 0.25>, ApplyForce=false, ApplyPosition=true, ApplyPositionFalloff=false)
   &AccelerationForce(
    Acceleration=&FN_FrontEndVector3Modifier(Gameplay=<0 0 -100>, Front End=&RandomRangeVector(Minimum=<-50 -50 0>, Maximum=<50 50 -100>))
    Coordinate Space=::Local
   )
   &SolveForcesAndVelocity(Clamp Velocity=false, Limit Acceleration=false)
  ]
  rendererClasses [&NiagaraMeshRendererProperties]
  renderers [{rendererClass=&NiagaraMeshRendererProperties, rendererIndex=0}]
 }
 SparklesLWE {rendererClasses [], renderers []}
 SpawnSmoke {
  emitterUpdateScript [
   &EmitterState()
   &SpawnBurst_Instantaneous(Spawn Count=9, Use Spawn Probability=false, UseLoopCountLimit=false)
  ]
  particleSpawnScript [
   &InitializeParticle(
    Color=&SelectLinearColorFromArray(
     Array Sampling Mode=::NewEnumerator1
     Direct Array Index=&RandomRangeInt(Minimum=0, Maximum=1)
     Color Selection Array=DataInterface() {propertyValues {ColorData [<rgba 0.1031 0.1031 0.1031 1>, <rgba 0.0573 0.0573 0.0573 1>]}}
     Random Seed=?
    )
    Lifetime=&RandomRangeFloat(Minimum=0.5, Maximum=1)
    Sprite Rotation Mode=::NewEnumerator2
    Sprite Size Mode=::NewEnumerator3
    Sprite UV Mode=::NewEnumerator3
    Uniform Sprite Size=&RandomRangeFloat(Minimum=15, Maximum=40)
    UsePositionOffset=false
    Write Material Random=true
   )
   &ShapeLocation(MaskSpawnGroup=false, Override Seed=false, Sphere Radius=5, UseOwnerScale=true)
   &AddVelocity(
    Module.Use Falloff Distance=false
    Origin Offset=<0 0 -1>
    Override Seed=false
    Velocity Mode=::NewEnumerator1
    Velocity Speed=&RandomRangeFloat(Minimum=100, Maximum=200)
   )
  ]
  particleUpdateScript [
   &ParticleState()
   &ScaleMeshSize(Scale Factor=&Multiply_VectorByFloat(
     Vector=<1 1 1>
     Float=&FloatFromCurve(
      FloatCurve=DataInterface() {propertyValues {Curve {
         defaultValue=1
         keys [<at 0 1 RCIM_Cubic>, <at 0.2225 1.7>, {arriveTangent=0.3858, interpMode=RCIM_Cubic, leaveTangent=0.3858, time=0.3582, value=1.85}, <at 1 2>]
        }}}
      Scale Curve=&RandomRangeFloat(Minimum=1, Maximum=1.3)
     )
    ))
   &ScaleSpriteSize(Uniform Curve Sprite Scale=DataInterface() {propertyValues {Curve {keys [
        {arriveTangent=1.1204, interpMode=RCIM_Cubic, leaveTangent=1.1204, tangentMode=RCTM_User, time=0, value=0.5}
        {tangentMode=RCTM_User, time=1, value=1}
       ]}}})
   &Drag(Drag=&RandomRangeFloat(Minimum=9, Maximum=13), Use Linear Drag=true, Use Rotational Drag=true)
   &InheritSourceMovement(
    Applied Position Scale=&Multiply_VectorByFloat(
     Vector=<1 1 1>
     Float=&FloatFromCurve(FloatCurve=DataInterface() {propertyValues {Curve {keys [{arriveTangent=-2.5112, interpMode=RCIM_Cubic, leaveTangent=-2.5112, tangentMode=RCTM_User, time=0, value=1}, <at 1 0.1>]}}})
    )
    ApplyForce=false
    ApplyPosition=true
    ApplyPositionFalloff=false
   )
   &SolveForcesAndVelocity(Clamp Velocity=false, Limit Acceleration=false)
   &DynamicMaterialParameters(
    Index 0 Param 1=&FloatFromCurve(FloatCurve=DataInterface() {propertyValues {Curve {defaultValue=1, keys [<at 0 1 RCIM_Cubic>, {arriveTangent=-1.9, interpMode=RCIM_Cubic, leaveTangent=-1.9, time=0.3, value=1}, <at 1 -0.9>]}}})
    Index 0 Param 4=0.5
    Module.Index0Float0Write=true
    Module.Index0Float1Write=true
    Module.Index0Float2Write=false
    Module.Index0Float3Write=true
   )
   &SubUVAnimation(
    Random Start Frame=true
    "Sprite Renderer "=DataInterface() {propertyValues {SpriteRenderer=&self:SpawnSmoke_0.NiagaraSpriteRendererProperties_0}}
    Start Frame Offset:NiagaraInt32=&Multiply_Int(A=0, B=16)
    SubUV Animation Mode:ENiagara_SubUVLookupModeV2=::NewEnumerator1
    SubUV Curve=DataInterface() {propertyValues {Curve {keys [
        {arriveTangent=4.1785, interpMode=RCIM_Cubic, leaveTangent=4.1785, tangentMode=RCTM_User, time=0, value=0}
        {arriveTangent=1.1449, interpMode=RCIM_Cubic, leaveTangent=1.1449, tangentMode=RCTM_User, time=0.1242, value=0.7035}
        {arriveTangent=0.2086, interpMode=RCIM_Cubic, leaveTangent=0.2086, tangentMode=RCTM_User, time=1, value=1}
       ]}}}
    UseEndFrame=false
    UseStartFrame=false
   )
  ]
  rendererClasses [&NiagaraSpriteRendererProperties]
  renderers [{rendererClass=&NiagaraSpriteRendererProperties}]
 }
 ToonImpactSmoke {rendererClasses [], renderers []}
 ToonSmoke {bEnabled=false, rendererClasses [], renderers []}
}
summary {
 systemName=NS_Emote_RubySnapper_CrowAppear
 emitters [
  {emitterName=Feathers, rendererClasses [&NiagaraMeshRendererProperties]}
  {emitterName=AirglowLWE, rendererClasses []}
  {emitterName=SparklesLWE, rendererClasses []}
  {emitterName=ToonImpactSmoke, rendererClasses []}
  {bEnabled=false, emitterName=ToonSmoke, rendererClasses []}
  {emitterName=SpawnSmoke, rendererClasses [&NiagaraSpriteRendererProperties]}
 ]
 userVariables [
  User.bIsFrontEnd {type=NiagaraBool}
  User.bIsFrontEndPreview {type=NiagaraBool}
 ]
}
