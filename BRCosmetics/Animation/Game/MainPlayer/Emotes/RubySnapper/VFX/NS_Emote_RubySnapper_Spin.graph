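# Niagara system for the spin effect of the RubySnapper emote, built around the RubySnapper gadget prop.
# Only Spin_Meshes is enabled, and it lists no modules or renderers here.
# Disabled emitters hold a time-of-day tinted anim-trail ribbon and two mesh bursts at the FX_Spin sockets, one rotating.
# A disabled sprite emitter places lights on rings around skeletal mesh samples, blinking alternately, then in unison, via HLSL sine expressions.
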NiagaraSystem /BRCosmetics/Animation/Game/MainPlayer/Emotes/RubySnapper/VFX/NS_Emote_RubySnapper_Spin
dependencies {
 usedDataInterfaces [
  &NiagaraDataInterfaceColorCurve
  &NiagaraDataInterfaceSocketReader
  &NiagaraDataInterfaceCurve
  &NiagaraDataInterfaceArrayFloat
  &NiagaraDataInterfaceSkeletalMesh
 ]
 usedDynamicInputs []
 usedModules [
  &self:SystemSpawnScript.SystemScriptSource.SystemScriptGraph.NiagaraNodeAssignment_0.SetVariables_#1
  &self:SystemSpawnScript.SystemScriptSource.SystemScriptGraph.NiagaraNodeAssignment_3.SetVariables_#2
  &SystemState
  &Timeline
  &self:SystemSpawnScript.SystemScriptSource.SystemScriptGraph.NiagaraNodeAssignment_1.SetVariables_#3
 ]
}
emitters {
 Empty {bEnabled=false, rendererClasses [], renderers []}
 FN_AnimTrail {
  bEnabled=false
  emitterSpawnScript [&FN_RibbonIDOffsetUtility()]
  emitterUpdateScript [
   &EmitterState(Life Cycle Mode=::NewEnumerator1, Loop Behavior=::NewEnumerator1, Loop Duration=3, UseLoopDelay=false)
   &FN_SpawnAnimTrail()
  ]
  particleSpawnScript [
   &FN_InitializeAnimTrailBoneIndecies()
   &FN_InitializeAnimTrailBoneIndecies(AnimTrail Selection=1, First Bone or Socket Index=2, Second Bone or Socket Index=3) {enabled=false, module=FN_InitializeAnimTrailBoneIndecies001}
   &FN_InitializeAnimTrail(
    Color=&FNTimeOfDayColor(
     Night Color=<rgba 0.0888 0.1178 0.1372 0.7>
     Morning Color=<rgba 0.2291 0.2349 0.2413 0.7>
     Day Color=<rgba 0.2555 0.2619 0.2691 0.7>
     Evening Color=<rgba 0.2691 0.2143 0.2289 0.7>
    )
    Lifetime=1
   )
   &AddVelocity(Velocity=<0 400 0>) {enabled=false}
  ]
  particleUpdateScript [
   &SolveForcesAndVelocity(Clamp Velocity=false, Limit Acceleration=false)
   &FN_UpdateAnimTrail() {module=FN_UpdateAnimTrail001}
   &ScaleColor(Scale Alpha=$System.MotionBlurOpacity, ScaleA=true, ScaleRGB=true)
   &Color(Color=&ColorFromCurve(ColorCurve=DataInterface() {propertyValues {
       AlphaCurve {keys [<at 0 1 RCIM_Cubic>,<at 1 0 RCIM_Cubic>]}
       BlueCurve {keys [
         <at 0 1 RCIM_Cubic>
         <at 0.2417 0>
         {arriveTangent=1.0679, interpMode=RCIM_Cubic, leaveTangent=1.0679, time=0.5125, value=0}
         <at 0.7984 0.5945>
         <at 1 1>
        ]}
       GreenCurve {keys [
         <at 0 0 RCIM_Cubic>
         <at 0.2417 0.188>
         {arriveTangent=0.9067, interpMode=RCIM_Cubic, leaveTangent=0.9067, time=0.5125, value=0.9459}
         <at 0.7984 0.6927>
         <at 1 0>
        ]}
       RedCurve {keys [
         <at 0 0.305 RCIM_Cubic>
         <at 0.2417 1>
         {arriveTangent=-1.7964, interpMode=RCIM_Cubic, leaveTangent=-1.7964, time=0.5125, value=1}
         <at 0.7984 0>
         <at 1 0.305>
        ]}
      }})) {enabled=false}
  ]
  rendererClasses [&NiagaraRibbonRendererProperties]
  renderers [
   {rendererClass=&NiagaraRibbonRendererProperties}
   {rendererClass=&NiagaraRibbonRendererProperties, rendererIndex=1}
   {rendererClass=&NiagaraRibbonRendererProperties, rendererIndex=2}
  ]
 }
 Minimal {
  bEnabled=false
  emitterUpdateScript [
   &EmitterState()
   &SpawnBurst_Instantaneous(Use Spawn Probability=false, UseLoopCountLimit=false)
  ]
  particleSpawnScript [
   &InitializeParticle(
    Color Mode=::NewEnumerator0
    Lifetime=1.5
    Mesh Scale Mode=::NewEnumerator3
    Mesh Uniform Scale=0.5
    UsePositionOffset=false
    Write Material Random=true
   )
   &SocketLocation(
    SamplePosOffset=false
    Socket Reader=DataInterface() {propertyValues {
      AttachComponentClass=null
      AttachComponentTag=None
      EditorPreviewAsset=/BRCosmetics/Animation/Props/Gadget_RubySnapper/Meshes/RubySnapper_Gadget
      SourceActorRef=null
      SourceAsset=null
      SourceMode=Default
      bRequireCurrentFrameData=true
      bUpdateSocketsPerFrame=true
      FilteredSockets [FX_Spin_0_Socket, FX_Spin_1_Socket]
      ObjectParameterBinding {parameter {name=None, typeDefHandle {registeredTypeIndex=1}, varData []}}
     }}
    Socket Sampling Mode=::NewEnumerator3
    VelocitySpeedLimit=false
   )
  ]
  particleUpdateScript [
   &ParticleState()
   &UpdateMeshOrientation(
    Rotation Rate=&FloatFromCurve(FloatCurve=DataInterface() {propertyValues {Curve {keys [<at 0 1 RCIM_Cubic>,<at 1 0>]}}}, Scale Curve=-6)
    Rotation Vector=<0 1 0>
   )
   &DynamicMaterialParameters(
    Index 0 Param 1=1
    Index 0 Param 2=&FloatFromCurve(FloatCurve=DataInterface() {propertyValues {Curve {keys [<at 0 1>,<at 1 0>]}}})
    Index 2 Param 1=&RandomRangeFloat(Minimum=0, Maximum=1)
    Index 2 Param 4=-0.1
    Module.Index0Float0Write=false
    Module.Index0Float1Write=false
    Module.Index0Float2Write=false
    Module.Index0Float3Write=false
    Module.Index2Float0Write=true
    Module.Index2Float1Write=false
    Module.Index2Float2Write=false
    Module.Index2Float3Write=true
    Write Parameter Index 2=::NewEnumerator1
   )
   &ScaleColor(
    Scale Alpha=&FloatFromCurve(
     FloatCurve=DataInterface() {propertyValues {Curve {keys [
         <at 0 0 RCIM_Cubic>
         {arriveTangentWeight=0.1026, interpMode=RCIM_Cubic, leaveTangentWeight=0.3043, tangentMode=RCTM_User, time=0.1, value=1}
         <at 1 0>
        ]}}}
     Scale Curve=2
    )
    Scale RGB=<0.5 0.5 1>
    ScaleA=true
    ScaleRGB=true
   )
  ]
  rendererClasses [&NiagaraMeshRendererProperties]
  renderers [
   {rendererClass=&NiagaraMeshRendererProperties, rendererIndex=0}
   {rendererClass=&NiagaraMeshRendererProperties, rendererIndex=1}
  ]
 }
 Minimal001 {
  bEnabled=false
  emitterUpdateScript [
   &EmitterState()
   &SpawnBurst_Instantaneous(Use Spawn Probability=false, UseLoopCountLimit=false)
  ]
  particleSpawnScript [&InitializeParticle(
    Color Mode=::NewEnumerator0
    Lifetime=1.5
    Mesh Scale Mode=::NewEnumerator3
    Mesh Uniform Scale=0.4
    UsePositionOffset=false
    Write Material Random=true
   )]
  particleUpdateScript [
   &ParticleState()
   &SocketLocation(
    SamplePosOffset=false
    Socket Reader=DataInterface() {propertyValues {
      AttachComponentClass=null
      AttachComponentTag=None
      EditorPreviewAsset=&RubySnapper_Gadget
      SourceActorRef=null
      SourceAsset=null
      SourceMode=Default
      bRequireCurrentFrameData=true
      bUpdateSocketsPerFrame=true
      FilteredSockets [FX_Spin_0_Socket, FX_Spin_1_Socket]
      ObjectParameterBinding {parameter {name=None, typeDefHandle {registeredTypeIndex=1}, varData []}}
     }}
    Socket Sampling Mode=::NewEnumerator3
    VelocitySpeedLimit=false
   )
   &ScaleColor(
    Scale Alpha=&FloatFromCurve(
     FloatCurve=DataInterface() {propertyValues {Curve {keys [
         <at 0 0 RCIM_Cubic>
         {arriveTangentWeight=0.1026, interpMode=RCIM_Cubic, leaveTangentWeight=0.3043, tangentMode=RCTM_User, time=0.05, value=1}
         <at 1 0>
        ]}}}
     Scale Curve=10
    )
    ScaleA=true
    ScaleRGB=true
   )
  ]
  rendererClasses [&NiagaraMeshRendererProperties]
  renderers [
   {rendererClass=&NiagaraMeshRendererProperties, rendererIndex=0}
   {rendererClass=&NiagaraMeshRendererProperties, rendererIndex=1}
  ]
 }
 Minimal002 {
  bEnabled=false
  emitterUpdateScript [
   &EmitterState(Life Cycle Mode=::NewEnumerator1, Loop Behavior=::NewEnumerator1, Loop Duration Mode=::NewEnumerator1, UseLoopDelay=false)
   &SpawnBurst_Instantaneous(Spawn Count=48, Use Spawn Probability=true, UseLoopCountLimit=false)
  ]
  particleSpawnScript [
   &InitializeParticle(
    Color=<rgba 5 4.7416 4.3049 0.5>
    Lifetime=12
    Sprite Size Mode=::NewEnumerator3
    Uniform Sprite Size=8
    UsePositionOffset=false
    Write Material Random=true
   )
   &CameraOffset(Camera Offset Amount=5)
   /SetVariables(Particles.IsEven=HlslExpression() {hlslExpression="fmod(Particles.ID.Index, 2) == 0 ? true : false"}) {isSetParameters=true}
  ]
  particleUpdateScript [
   &ParticleState()
   &SkeletalMeshLocation(Bone Sampling Mode=::NewEnumerator4, MaskSpawnGroup=false, Sampled Position Offset=<0 15 0>, Skeletal Mesh=$System.Skeletal Mesh Reader)
   &ShapeLocation(
    MaskSpawnGroup=false
    Offset=&MakeCustomVectorFromBool(
     Bool=&SetBoolByFloatComparison(A=&MakeFloatFromInt(Int=&ReturnParticleIDAttributes()), B=11)
     True Vector=<0 4 0>
     False Vector=<0 -7 0>
    )
    Offset Coordinate Space=::Local
    Offset Mode=::NewEnumerator0
    Override Seed=false
    Ring / Disc Distribution Mode:ENiagaraSphereDistributionMode=::NewEnumerator1
    Ring Radius=33.5
    Rotation Axis=<0 1 0>
    Rotation Mode=::NewEnumerator1
    Shape Origin=$Output.SkeletalMeshLocation.SampledPositionWithoutOffset
    Shape Primitive=::NewEnumerator4
    U Position=&Add_Float(
     A=0
     B=&SelectFloatFromArray(
      Direct Array Index=&MakeIntFromFloat(Float=&FloatFromCurve(
        FloatCurve=DataInterface() {propertyValues {Curve {defaultValue=1, keys [<at 0 0>, {tangentMode=RCTM_User, time=0, value=0}, <at 0.5 1 RCIM_Constant>, <at 0.5 0>, <at 1 1 RCIM_Constant>]}}}
        CurveIndex=&ReturnNormalizedExecIndex()
        Scale Curve=12
       ))
      Float Selection Array=DataInterface() {propertyValues {FloatData [0,0.06,0.117,0.18,0.388,0.445,0.502,0.56,0.62,0.68,0.88,0.94]}}
     )
    )
    UseOwnerScale=true
   )
   &ScaleColor(
    Scale Alpha=&MakeCustomFloatFromBool(
     Bool=HlslExpression() {hlslExpression="(Emitter.Age > 3.5) ? true : false /* Custom HLSL! */"}
     True Float=HlslExpression() {hlslExpression="clamp(ceil(sin(Engine.Time * 20)*2), 0, 2)"}
     False Float=&Multiply_Float(
      A=HlslExpression() {hlslExpression="Particles.IsEven ? floor(sin(Engine.Time * 15) + 1) : floor(sin((Engine.Time + (PI / 2)) * 15) + 1)"}
      B=1
     )
    )
    Scale RGB=&Multiply_VectorByFloat(Vector=<1 1 1>, Float=1)
    ScaleA=true
    ScaleRGB=false
   )
  ]
  rendererClasses [&NiagaraSpriteRendererProperties]
  renderers [{rendererClass=&NiagaraSpriteRendererProperties}]
 }
 Spin_Meshes {rendererClasses [], renderers []}
}
summary {
 systemName=NS_Emote_RubySnapper_Spin
 emitters [
  {bEnabled=false, emitterName=FN_AnimTrail, rendererClasses [&NiagaraRibbonRendererProperties]}
  {bEnabled=false, emitterName=Empty, rendererClasses []}
  {bEnabled=false, emitterName=Minimal, rendererClasses [&NiagaraMeshRendererProperties]}
  {bEnabled=false, emitterName=Minimal001, rendererClasses [&NiagaraMeshRendererProperties]}
  {emitterName=Spin_Meshes, rendererClasses []}
  {bEnabled=false, emitterName=Minimal002, rendererClasses [&NiagaraSpriteRendererProperties]}
 ]
 userVariables [
  User.bIsSwinging {type=NiagaraBool}
  User.RibbonIDOffset {type=NiagaraInt32}
 ]
}
