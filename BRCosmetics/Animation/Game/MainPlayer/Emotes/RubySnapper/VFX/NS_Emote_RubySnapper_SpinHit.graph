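# Niagara effect for the RubySnapper emote's spin hit, pairing a short impact flash with a motion trail.
# Minimal001 bursts a few short-lived SubUV sprites after a brief delay through three sprite renderers, animating separate ring and elongated size attributes.
# Speed_Lines002 spawns a fading ribbon by distance traveled, tracing the emitter's movement.
# The particle burst, mesh speed-line and both sparkle emitters are disabled, including the only logic reading User.WorkoutSelect.
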NiagaraSystem /BRCosmetics/Animation/Game/MainPlayer/Emotes/RubySnapper/VFX/NS_Emote_RubySnapper_SpinHit
dependencies {
 usedDataInterfaces [
  &NiagaraDataInterfaceArrayColor
  &NiagaraDataInterfaceColorCurve
  &NiagaraDataInterfaceCurve
  &NiagaraDataInterfaceVector2DCurve
  &NiagaraDataInterfaceArrayFloat
  &NiagaraDataInterfaceArrayFloat3
  &NiagaraDataInterfaceSpriteRendererInfo
  &NiagaraDataInterfaceSkeletalMesh
 ]
 usedDynamicInputs [&MakeFloatFromVector2d]
 usedModules [
  &self:SystemSpawnScript.SystemScriptSource.SystemScriptGraph.NiagaraNodeAssignment_1.SetVariables_#1
  &SystemState
 ]
}
emitters {
 Minimal {
  bEnabled=false
  emitterUpdateScript [
   &EmitterState(Life Cycle Mode=::NewEnumerator1, Loop Behavior=::NewEnumerator1, Loop Delay=0.1, Loop Duration=1, UseLoopDelay=true)
   &SpawnBurst_Instantaneous(Spawn Count=30, Use Spawn Probability=false, UseLoopCountLimit=false)
  ]
  particleSpawnScript [
   &InitializeParticle(
    Color=&SelectLinearColorFromArray(
     Array Sampling Mode=::NewEnumerator1
     Direct Array Index=&RandomRangeInt(Minimum=0, Maximum=1)
     Color Selection Array=DataInterface() {propertyValues {ColorData [<rgba 0 0 0 1>, <rgba 1 1 1 1>]}}
     Random Seed=?
    )
    Lifetime Max=1
    Lifetime Min=0.3
    Lifetime Mode=::NewEnumerator1
    Sprite Rotation Angle=45
    Sprite Rotation Mode=::NewEnumerator0
    Sprite Size Mode=::NewEnumerator1
    Uniform Sprite Size Max=15
    Uniform Sprite Size Min=8
    UsePositionOffset=false
    Write Material Random=true
   )
   &ShapeLocation(MaskSpawnGroup=false, Override Seed=false, Sphere Radius=1, UseOwnerScale=true)
   &AddVelocity(Module.Use Falloff Distance=false, Origin Offset=<0 -1 0>, Override Seed=false, Velocity Mode=::NewEnumerator1, Velocity Speed=500)
  ]
  particleUpdateScript [
   &ParticleState()
   &Drag(Drag=5, Use Linear Drag=true, Use Rotational Drag=true)
   &AccelerationForce(Acceleration=<0 0 200>)
   &SolveForcesAndVelocity(Clamp Velocity=false, Limit Acceleration=false)
   &ScaleColor(
    Scale Mode=1
    Scale RGBA=&Vector4FromCurve(Vector4Curve=DataInterface() {propertyValues {
       AlphaCurve {defaultValue=0, keys [<at 0 0>,<at 0.2 1>,<at 1 0>]}
       BlueCurve {keys [<at 0 1>,<at 1 1>]}
       GreenCurve {defaultValue=1, keys [<at 0 1>,<at 1 1>]}
       RedCurve {keys [<at 0 1>,<at 1 1>]}
      }})
    ScaleRGBA=true
   )
   &ScaleSpriteSize(Uniform Curve Sprite Scale=DataInterface() {propertyValues {Curve {defaultValue=0, keys [<at 0 0>,<at 0.105 1>,<at 1 0>]}}})
  ]
  rendererClasses [&NiagaraSpriteRendererProperties]
  renderers [{rendererClass=&NiagaraSpriteRendererProperties}]
 }
 Minimal001 {
  emitterUpdateScript [
   &EmitterState(Life Cycle Mode=::NewEnumerator1, Loop Behavior=::NewEnumerator1, Loop Delay=0.09, Loop Duration=1, UseLoopDelay=true)
   &SpawnBurst_Instantaneous(Spawn Count=6, Use Spawn Probability=false, UseLoopCountLimit=false)
  ]
  particleSpawnScript [
   /SetVariables(Particles.VisibilityTag=&ReturnParticleIDAttributes()) {isSetParameters=true}
   &InitializeParticle(
    Color Mode=::NewEnumerator0
    Lifetime=&RandomRangeFloat(Minimum=0.15, Maximum=0.25)
    Sprite Rotation Mode=::NewEnumerator2
    Sprite Size Mode=::NewEnumerator1
    Sprite UV Mode=::NewEnumerator1
    Uniform Sprite Size Max=50
    Uniform Sprite Size Min=40
    UsePositionOffset=false
    Write Material Random=true
   )
   &SubUVAnimation(
    Evaluation Type=::NewEnumerator0
    "Sprite Renderer "=DataInterface() {propertyValues {SpriteRenderer=&self:Minimal001.Renderer}}
    SubUV Animation Mode:ENiagara_SubUVLookupModeV2=::NewEnumerator2
    UseEndFrame=false
    UseStartFrame=false
   )
   /SetVariables() {isSetParameters=true}
   &InheritVelocity(Inherited Velocity Speed Limit=20, Speed Limit=true, UseSpeedThreshold=false)
   &CameraOffset(Camera Offset Amount=50)
  ]
  particleUpdateScript [
   &ParticleState()
   &ScaleSpriteSize(Uniform Curve Sprite Scale=DataInterface() {propertyValues {Curve {keys [
        {arriveTangent=7.7405, interpMode=RCIM_Cubic, leaveTangent=7.7405, tangentMode=RCTM_User, time=0, value=0}
        {interpMode=RCIM_Cubic, tangentMode=RCTM_User, time=0.2, value=1}
        <at 1 0>
       ]}}})
   /SetVariables(
    Particles.RingSpriteSize:Vector2f=&Multiply_Vector2D_ByFloat(
     Float=&FloatFromCurve(FloatCurve=DataInterface() {propertyValues {Curve {keys [{arriveTangent=0.8668, interpMode=RCIM_Cubic, leaveTangent=0.8668, tangentMode=RCTM_User, time=0, value=1}, <at 1 1.5>]}}})
     Vector2D=<50 50>
    )
    Particles.SpriteSize2=&Vector2DFromCurve(
     Vector2Curve=DataInterface() {propertyValues {
       XCurve {keys [
         {arriveTangent=8.1517, interpMode=RCIM_Cubic, leaveTangent=8.1517, tangentMode=RCTM_User, time=0, value=0.2}
         {arriveTangent=0.2469, interpMode=RCIM_Cubic, leaveTangent=0.2469, tangentMode=RCTM_User, time=0.15, value=1}
         <at 1 0>
        ]}
       YCurve {keys [
         {arriveTangent=8.1517, interpMode=RCIM_Cubic, leaveTangent=8.1517, tangentMode=RCTM_User, time=0, value=0.5}
         {arriveTangent=-0.0461, interpMode=RCIM_Cubic, leaveTangent=-0.0461, tangentMode=RCTM_User, time=0.15, value=1}
         <at 1 0>
        ]}
      }}
     ScaleCurve=&RandomRangeVector2D(Minimum=<30 85>, Maximum=<40 95>)
    )
   ) {isSetParameters=true}
   &Drag(Drag=5, Use Linear Drag=true, Use Rotational Drag=false)
   &SolveForcesAndVelocity(Clamp Velocity=false, Limit Acceleration=false)
   &ScaleColor(
    Scale Alpha=&FloatFromCurve(FloatCurve=DataInterface() {propertyValues {Curve {keys [{interpMode=RCIM_Cubic, tangentMode=RCTM_User, time=0, value=1}, <at 1 0>]}}})
    ScaleA=true
    ScaleRGB=false
   )
  ]
  rendererClasses [&NiagaraSpriteRendererProperties]
  renderers [
   {rendererClass=&NiagaraSpriteRendererProperties}
   {rendererClass=&NiagaraSpriteRendererProperties, rendererIndex=1}
   {rendererClass=&NiagaraSpriteRendererProperties, rendererIndex=2}
  ]
 }
 SparklesLWE {bEnabled=false, rendererClasses [], renderers []}
 SparklesLWE001 {bEnabled=false, rendererClasses [], renderers []}
 Speed_Lines {
  bEnabled=false
  emitterUpdateScript [
   &EmitterState(Life Cycle Mode=::NewEnumerator1, Loop Duration=0.95, UseLoopDelay=false)
   &SpawnRate(
    SpawnRate=&SelectFloatFromArray(
     Direct Array Index=&MakeIntFromFloat(Float=$User.WorkoutSelect)
     Float Selection Array=DataInterface() {propertyValues {FloatData [0,0,0,24,0]}}
    )
    Use Spawn Probability=false
   ) {module=SpawnRate001}
  ]
  particleSpawnScript [
   &InitializeParticle(
    Color=&RandomRangeLinearColor(Minimum=<rgba 0 0 0 20>, Maximum=<rgba 0.0521 0.0521 0.0521 25>)
    Lifetime Max=0.168
    Lifetime Min=0.114
    Lifetime Mode=::NewEnumerator1
    Mesh Scale Max=<0.8 0.8 0.5>
    Mesh Scale Min=<0.8 0.8 0.5>
    Mesh Scale Mode=::NewEnumerator2
    Position Mode=::NewEnumerator1
    UsePositionOffset=false
    Write Material Random=true
   )
   &SkeletalMeshLocation(Bone / Socket Index=5, Bone Sampling Mode=::NewEnumerator4, MaskSpawnGroup=false, Skeletal Mesh=$System.Skeletal Mesh)
   &ShapeLocation(
    Box Size=<5 5 8>
    MaskSpawnGroup=false
    Offset=&SelectVectorFromArray(
     Direct Array Index=&RandomRangeInt(Minimum=0, Maximum=1)
     Vector Selection Array=DataInterface() {propertyValues {FloatData [<-16 0 113>, <-16 -40 113>]}}
    )
    Offset Coordinate Space=::Local
    Offset Mode=::NewEnumerator0
    Override Seed=false
    Shape Origin=$Particles.SkeletalMeshLocation.SampledPosition
    Shape Primitive=::NewEnumerator2
    Surface Only=false
    UseOwnerScale=true
   )
   &AddVelocity(Velocity=&Multiply_Vector(
     A=&MakeVector(
      X=0
      Y=0
      Z=&SelectFloatFromArray(
       Direct Array Index=&MakeIntFromFloat(Float=$User.WorkoutSelect)
       Float Selection Array=DataInterface() {propertyValues {FloatData [100,0,0,100]}}
      )
     )
     B=&VectorFromFloat(Value=&SelectFloatFromArray(
       Array Sampling Mode=::NewEnumerator0
       Direct Array Index=?
       Float Selection Array=DataInterface() {propertyValues {FloatData [-1,1]}}
       Random Seed=0
      ))
    )) {enabled=false, module=AddVelocity001}
   &CameraOffset(Camera Offset Amount=&FN_FrontEndFloatModifier(Gameplay=0, Front End=-30))
  ]
  particleUpdateScript [
   &ParticleState()
   &ScaleMeshSize(Scale Factor=&MakeVector(
     X=&FloatFromCurve(FloatCurve=DataInterface() {propertyValues {Curve {keys [<at 0 0.4>,<at 1 1>]}}})
     Y=&FloatFromCurve(FloatCurve=DataInterface() {propertyValues {Curve {defaultValue=1, keys [{tangentMode=RCTM_User, time=0, value=1}, {tangentMode=RCTM_User, time=1, value=0.4}]}}})
     Z=&FloatFromCurve(FloatCurve=DataInterface() {propertyValues {Curve {defaultValue=1, keys [{tangentMode=RCTM_User, time=0, value=1}, {tangentMode=RCTM_User, time=1, value=1}]}}})
    ))
   &Drag(Drag=5, Use Linear Drag=true, Use Rotational Drag=false)
   &SolveForcesAndVelocity(Clamp Velocity=false, Limit Acceleration=false)
   &ScaleColor(
    Scale Alpha=&FloatFromCurve(FloatCurve=DataInterface() {propertyValues {Curve {keys [
         {interpMode=RCIM_Cubic, tangentMode=RCTM_User, time=0.1552, value=0}
         {interpMode=RCIM_Cubic, tangentMode=RCTM_User, time=0.5, value=1}
         {tangentMode=RCTM_User, time=1, value=0}
        ]}}})
    ScaleA=true
    ScaleRGB=true
   )
   &ScaleSpriteSize(
    Non-Uniform Curve Sprite Scale=DataInterface() {propertyValues {
      XCurve {keys [{tangentMode=RCTM_User, time=0, value=0.5}, {tangentMode=RCTM_User, time=1, value=1.5}]}
      YCurve {keys [{tangentMode=RCTM_User, time=0, value=1}, {tangentMode=RCTM_User, time=1, value=0.5}]}
     }}
    Scale Sprite Size Mode=::NewEnumerator3
   )
   /SetVariables(Particles.New Position:NiagaraPosition=&Add_Vector(
     A=Linked() {linkedVariable {name=Particles.Position, type=NiagaraPosition}}
     B=Linked() {linkedVariable {name=Engine.Emitter.SimulationPosition, type=NiagaraPosition}}
    )) {enabled=false, isSetParameters=true}
  ]
  rendererClasses [&NiagaraMeshRendererProperties]
  renderers [{rendererClass=&NiagaraMeshRendererProperties, rendererIndex=0}]
 }
 Speed_Lines002 {
  emitterUpdateScript [
   &EmitterState(Life Cycle Mode=::NewEnumerator1, Loop Behavior=::NewEnumerator1, Loop Duration=0.31, UseLoopDelay=false)
   &SpawnRate(SpawnRate=50, Use Spawn Probability=false) {enabled=false}
   &SpawnPerUnit(Module.Use Max Movement Threshold=true, Module.Use Movement Tolerance=false, Spawn Spacing=15, Use Spawn Probability=false)
  ]
  particleSpawnScript [
   &InitializeParticle(
    Lifetime=0.2
    Position Mode=::NewEnumerator1
    Ribbon Width=10
    Ribbon Width Mode=::NewEnumerator1
    UsePositionOffset=false
    Write Material Random=true
   )
   &SkeletalMeshLocation(Bone / Socket Index=5, Bone Sampling Mode=::NewEnumerator4, MaskSpawnGroup=false, Skeletal Mesh=$System.Skeletal Mesh) {enabled=false}
   &ShapeLocation(MaskSpawnGroup=false, Override Seed=false, Shape Origin=$Engine.Emitter.SimulationPosition, Sphere Radius=0, UseOwnerScale=true)
   &AddVelocity(Velocity=<0 0 5>) {enabled=false, module=AddVelocity001}
   &CameraOffset(Camera Offset Amount=&FN_FrontEndFloatModifier(Gameplay=0, Front End=-30)) {enabled=false}
  ]
  particleUpdateScript [
   &ParticleState()
   &ScaleMeshSize(Scale Factor=&MakeVector(
     X=&FloatFromCurve(FloatCurve=DataInterface() {propertyValues {Curve {keys [<at 0 0.4>,<at 1 1>]}}})
     Y=&FloatFromCurve(FloatCurve=DataInterface() {propertyValues {Curve {defaultValue=1, keys [{tangentMode=RCTM_User, time=0, value=1}, {tangentMode=RCTM_User, time=1, value=0.4}]}}})
     Z=&FloatFromCurve(FloatCurve=DataInterface() {propertyValues {Curve {defaultValue=1, keys [{tangentMode=RCTM_User, time=0, value=1}, {tangentMode=RCTM_User, time=1, value=1}]}}})
    )) {enabled=false}
   &Drag(Drag=5, Use Linear Drag=true, Use Rotational Drag=false) {enabled=false}
   &AccelerationForce(Acceleration=<0 0 50>) {enabled=false}
   &SolveForcesAndVelocity(Clamp Velocity=false, Limit Acceleration=false)
   &ScaleColor(
    Scale Alpha=&FloatFromCurve(FloatCurve=DataInterface() {propertyValues {Curve {keys [
         {interpMode=RCIM_Cubic, tangentMode=RCTM_User, time=0.1552, value=0}
         {interpMode=RCIM_Cubic, tangentMode=RCTM_User, time=0.2, value=1}
         {tangentMode=RCTM_User, time=1, value=0}
        ]}}})
    ScaleA=true
    ScaleRGB=true
   )
   &ScaleSpriteSize(
    Non-Uniform Curve Sprite Scale=DataInterface() {propertyValues {
      XCurve {keys [{tangentMode=RCTM_User, time=0, value=0.5}, {tangentMode=RCTM_User, time=1, value=1.5}]}
      YCurve {keys [{tangentMode=RCTM_User, time=0, value=1}, {tangentMode=RCTM_User, time=1, value=0.5}]}
     }}
    Scale Sprite Size Mode=::NewEnumerator3
   ) {enabled=false}
  ]
  rendererClasses [&NiagaraRibbonRendererProperties]
  renderers [{rendererClass=&NiagaraRibbonRendererProperties}]
 }
}
summary {
 systemName=NS_Emote_RubySnapper_SpinHit
 emitters [
  {bEnabled=false, emitterName=SparklesLWE, rendererClasses []}
  {bEnabled=false, emitterName=Minimal, rendererClasses [&NiagaraSpriteRendererProperties]}
  {emitterName=Speed_Lines002, rendererClasses [&NiagaraRibbonRendererProperties]}
  {bEnabled=false, emitterName=Speed_Lines, rendererClasses [&NiagaraMeshRendererProperties]}
  {bEnabled=false, emitterName=SparklesLWE001, rendererClasses []}
  {emitterName=Minimal001, rendererClasses [&NiagaraSpriteRendererProperties]}
 ]
 userVariables [
  User.bIsFrontEnd {type=NiagaraBool}
  User.bIsFrontEndPreview {type=NiagaraBool}
  User.WorkoutSelect {type=NiagaraFloat, defaultValue=0, description=""}
 ]
}
