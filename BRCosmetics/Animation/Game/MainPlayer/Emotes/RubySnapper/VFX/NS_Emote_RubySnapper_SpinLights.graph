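# Niagara effect for the RubySnapper emote that shows warm yellow sprite lights arranged in rings.
# A single burst of 24 sprites is split by particle ID between two local Y offsets.
# Ring placement comes from a curve-driven index into twelve hand-authored, unevenly spaced positions.
# Custom HLSL pulses alpha on sine waves phase-shifted between even and odd particles, then holds it steady once emitter age passes 3.5 seconds.
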NiagaraSystem /BRCosmetics/Animation/Game/MainPlayer/Emotes/RubySnapper/VFX/NS_Emote_RubySnapper_SpinLights
dependencies {
 usedDataInterfaces [&NiagaraDataInterfaceCurve, &NiagaraDataInterfaceArrayFloat]
 usedDynamicInputs []
 usedModules [&SystemState]
}
emitters {Minimal {
  emitterUpdateScript [
   &EmitterState(Life Cycle Mode=::NewEnumerator1, Loop Behavior=::NewEnumerator1, Loop Duration Mode=::NewEnumerator1, UseLoopDelay=false)
   &SpawnBurst_Instantaneous(Spawn Count=24, Use Spawn Probability=false, UseLoopCountLimit=false)
  ]
  particleSpawnScript [
   &InitializeParticle(
    Color=<rgba 1 0.6523 0.0635 1>
    Lifetime=10
    Sprite Size Mode=::NewEnumerator3
    Uniform Sprite Size=30
    UsePositionOffset=false
    Write Material Random=true
   )
   /SetVariables(Particles.IsEven=HlslExpression() {hlslExpression="fmod(Particles.ID.Index, 2) == 0 ? true : false"}) {isSetParameters=true}
   &ShapeLocation(
    MaskSpawnGroup=false
    Offset=&MakeCustomVectorFromBool(
     Bool=&SetBoolByFloatComparison(A=&MakeFloatFromInt(Int=&ReturnParticleIDAttributes()), B=11)
     True Vector=<0 5 0>
     False Vector=<0 -9 0>
    )
    Offset Coordinate Space=::Local
    Offset Mode=::NewEnumerator0
    Override Seed=false
    Ring / Disc Distribution Mode:ENiagaraSphereDistributionMode=::NewEnumerator1
    Ring Radius=34
    Rotation Angle=-0.03
    Rotation Axis=<0 1 0>
    Rotation Mode=::NewEnumerator1
    Shape Origin=$Engine.Emitter.SimulationPosition
    Shape Primitive=::NewEnumerator4
    U Position=&Add_Float(
     A=0
     B=&SelectFloatFromArray(
      Direct Array Index=&MakeIntFromFloat(Float=&FloatFromCurve(
        FloatCurve=DataInterface() {propertyValues {Curve {defaultValue=1, keys [<at 0 0>, {tangentMode=RCTM_User, time=0, value=0}, <at 0.5 1 RCIM_Constant>, <at 0.5 0>, <at 1 1 RCIM_Constant>]}}}
        CurveIndex=&ReturnNormalizedExecIndex()
        Scale Curve=12
       ))
      Float Selection Array=DataInterface() {propertyValues {FloatData [0,0.06,0.117,0.18,0.388,0.445,0.502,0.56,0.62,0.68,0.88,0.94]}}
     )
    )
    UseOwnerScale=true
   )
   &CameraOffset(Camera Offset Amount=4)
  ]
  particleUpdateScript [
   &ParticleState()
   &ScaleColor(
    Scale Alpha=&MakeCustomFloatFromBool(
     Bool=HlslExpression() {hlslExpression="(Emitter.Age > 3.5) ? true : false /* Custom HLSL! */"}
     True Float=1
     False Float=&Multiply_Float(A=HlslExpression() {hlslExpression="Particles.IsEven ? sin(Engine.Time * 22) + 1 : sin((Engine.Time + (PI / 2)) * 22) + 1"}, B=1)
    )
    Scale RGB=&Multiply_VectorByFloat(Vector=<1 1 1>, Float=1)
    ScaleA=true
    ScaleRGB=false
   ) {module=ScaleColor001}
  ]
  rendererClasses [&NiagaraSpriteRendererProperties]
  renderers [{rendererClass=&NiagaraSpriteRendererProperties}]
 }}
summary {
 systemName=NS_Emote_RubySnapper_SpinLights
 emitters [{emitterName=Minimal, rendererClasses [&NiagaraSpriteRendererProperties]}]
 userVariables []
}
